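# Physics asset for the Topple vehicle body cosmetic, giving the skeletal mesh a simplified physics shape.
# A single body on the root bone holds one box, longest along X and offset slightly along +X and raised along Z.
# It has no constraints and no other bodies, so the whole body is represented as one block with no articulated parts.
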PhysicsAsset /VehicleCosmetics/Bodies/Topple/Meshes/SK_Body_Topple_PhysicsDM
bodies {root [{shapeType=Box, extentX=444, extentY=216, extentZ=116, shapeName=root_box, center=<17 0 99>}]}
constraints []
