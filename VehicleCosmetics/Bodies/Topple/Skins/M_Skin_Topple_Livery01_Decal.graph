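# Masked, Default Lit material, named as the decal layer for the Livery01 skin of the Topple vehicle body cosmetic.
# Its graph holds a single texture sample of the livery decals texture, with no connection to any material output shown.
# Automatic setting of usage flags in the editor is turned off.
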Material /VehicleCosmetics/Bodies/Topple/Skins/M_Skin_Topple_Livery01_Decal
graph {
 properties {bAutomaticallySetUsageInEditor=false, blendMode=BLEND_Masked, usedShadingModels=Default Lit}
 nodes [TextureSample() {texture=&/T_skin_Topple_Livery01_Decals}]
}
