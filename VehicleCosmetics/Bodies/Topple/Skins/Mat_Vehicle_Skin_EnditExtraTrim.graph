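# Vehicle skin material for the Topple body that combines the Endit skin with an extra trim layer.
# The Endit skin and v2 trim base functions blend as material attributes, masked by the green channel of a Blightshield curvature pack texture.
# The blend feeds body normal LOD, then vehicle post FX.
# A separate damage to de-rez transition chain takes no input from that blend.
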Material /VehicleCosmetics/Bodies/Topple/Skins/Mat_Vehicle_Skin_EnditExtraTrim
graph {
 properties {
  bAutomaticallySetUsageInEditor=false
  bHasPixelAnimation=true
  bUseMaterialAttributes=true
  bUsedWithMeshDeformer=true
  bUsedWithSkeletalMesh=true
  usedShadingModels=Default Lit
 }
 nodes [
  n1 = MaterialFunctionCall() {materialFunction=/VehicleCosmetics/UniversalSkins/Endit/MF_Skin_Endit}
  n2 = MaterialFunctionCall() {materialFunction=/VehicleCosmetics/Materials/VehicleParents_Version2/Material_Functions/MF_Trim_Base_v2}
  n3 = TextureSampleParameter2D() {parameterName=CurvaturePack, texture=&/T_Topple_Body_Curvature_Blightshield}
  MaterialFunctionCall!(In=MaterialFunctionCall!(In=BlendMaterialAttributes(A=@n1.Result, B=@n2.Result, Alpha=@n3.G)) {materialFunction=/VehicleCosmetics/Materials/MaterialFunctions/MF_BodyNormalLOD}.Result) {materialFunction=/VehicleCosmetics/Materials/MaterialFunctions/MF_Vehicle_FNPostFX}
  MaterialFunctionCall(In=MaterialFunctionCall() {materialFunction=&MF_BaseVehicle_Damage}.Result) {materialFunction=&MF_VehicleDeRez_Transition}
 ]
}
